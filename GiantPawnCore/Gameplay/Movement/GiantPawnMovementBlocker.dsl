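; GiantPawnMovementBlocker is an actor with a box component that, judging by its names, stops giant pawn movement as a blocking volume.
; On BeginPlay it checks a ScalableFloat hotfix toggle for whether blocking volumes are enabled.
; When enabled, it optionally draws a long-lived yellow debug box matching the stop-movement box.
; When disabled, it turns off actor collision and destroys itself.
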(event EventBeginPlay
  ;[0.276 0.276 1] Begin Play
  (if (ScalableFloat|GetValueasBool (Variables|Default|GetEnableBlockingVolumeHotfix))
    (if (Variables|Default|GetDrawDebug)
      (Rendering|Debug|DrawDebugBox (Transformation|GetWorldLocation (Variables|Default|GetBox_StopMovement)) (Components|Box|GetScaledBoxExtent (Variables|Default|GetBox_StopMovement)) "(R=1.000000,G=0.900219,B=0.000000,A=1.000000)" (Transformation|GetWorldRotation (Variables|Default|GetBox_StopMovement)) 1e+04))
    (else
      (Collision|SetActorEnableCollision false)
      (Actor|DestroyActor))))

(fn ConstructionScript)
